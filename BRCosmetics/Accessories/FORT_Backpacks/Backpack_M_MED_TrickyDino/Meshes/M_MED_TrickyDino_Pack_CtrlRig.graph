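# Control Rig for the TrickyDino backpack that adds procedural corrective motion to its scarf and left foot bones.
# Parent constraints blend the scarf and knot base bones between pack and dyn_head.
# DeltaRotation nodes measure head and left-leg rotation against reference nulls via swing-twist, remapping one Euler axis to a clamped weight.
# Those weights drive rotation offsets on the scarf bones and left foot; the right leg has no corrective.
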ControlRigBlueprint /BRCosmetics/Accessories/FORT_Backpacks/Backpack_M_MED_TrickyDino/Meshes/M_MED_TrickyDino_Pack_CtrlRig
graphs [
 New Function_ContainedGraph {nodes [
   n1 = RigVMFunctionEntryNode()
   RigVMFunction_MathBoolToInteger = &RigVMFunction_MathBoolToInteger(Value=@n1.onlyTwist)
   RigVMFunction_MathQuaternionSwingTwist = &RigVMFunction_MathQuaternionSwingTwist(
    TwistAxis=@n1.twistAxis
    Input=&RigUnit_GetRelativeTransformForItem(
     Child=(Type=Bone,Name="None")
     Parent=(Type=Bone,Name="None")
     bChildInitial=False
     bParentInitial=False
     Child.Name=@n1.child
     Parent.Name=@n1.parent
     Parent.Type=@n1.parentType
    ).RelativeTransform.Rotation
   )
   RigVMFunction_MathQuaternionToEuler = &RigVMFunction_MathQuaternionToEuler(
    Value=RigVMDispatchNode(
     Values=((),())
     Index=@RigVMFunction_MathBoolToInteger.Result
     Values.0=@RigVMFunction_MathQuaternionSwingTwist.Swing
     Values.1=@RigVMFunction_MathQuaternionSwingTwist.Twist
    ) {template="DISPATCH_RigVMDispatch_SelectInt32(in Index,in Values,out Result)"}.Result
    RotationOrder=@n1.Rotation Order
   )
   > DISPATCH_RigVMDispatch_Print = RigVMDispatchNode(ScreenColor=(R=1,G=1,B=1,A=1), ScreenDuration=1.5, Enabled=@n1.debugRotation, Value=@RigVMFunction_MathQuaternionToEuler.Result) {template="DISPATCH_RigVMDispatch_Print(in Prefix,in Value,in Enabled,in ScreenDuration,in ScreenColor)"}
   Clamp = &RigVMFunction_MathIntClamp(Value=@n1.primAxis)
   DISPATCH_RigVMDispatch_SelectInt32 = RigVMDispatchNode(
    Values=(0,0,0)
    Index=@Clamp.Result
    Values.0=@RigVMFunction_MathQuaternionToEuler.Result.X
    Values.1=@RigVMFunction_MathQuaternionToEuler.Result.Y
    Values.2=@RigVMFunction_MathQuaternionToEuler.Result.Z
   ) {template="DISPATCH_RigVMDispatch_SelectInt32(in Index,in Values,out Result)"}
   Remap = &RigVMFunction_MathDoubleRemap(
    TargetMaximum=1
    TargetMinimum=0
    bClamp=False
    Value=@DISPATCH_RigVMDispatch_SelectInt32.Result
    SourceMaximum=@n1.max
    SourceMinimum=@n1.min
   )
   Multiply = &RigVMFunction_MathDoubleMul(A=&RigVMFunction_MathDoubleClamp(Value=@Remap.Result).Result, B=@n1.premult)
   > RigVMFunctionReturnNode(outputWeight=@Multiply.Result)
  ]}
 New Function_ContainedGraph {nodes [
   n1 = RigVMFunctionEntryNode()
   RigVMFunction_MathBoolToInteger = &RigVMFunction_MathBoolToInteger(Value=@n1.onlyTwist)
   RigVMFunction_MathQuaternionSwingTwist = &RigVMFunction_MathQuaternionSwingTwist(
    TwistAxis=@n1.twistAxis
    Input=&RigUnit_GetRelativeTransformForItem(
     Child=(Type=Bone,Name="None")
     Parent=(Type=Bone,Name="None")
     bChildInitial=False
     bParentInitial=False
     Child.Name=@n1.child
     Parent.Name=@n1.parent
     Parent.Type=@n1.parentType
    ).RelativeTransform.Rotation
   )
   RigVMFunction_MathQuaternionToEuler = &RigVMFunction_MathQuaternionToEuler(
    Value=RigVMDispatchNode(
     Values=((),())
     Index=@RigVMFunction_MathBoolToInteger.Result
     Values.0=@RigVMFunction_MathQuaternionSwingTwist.Swing
     Values.1=@RigVMFunction_MathQuaternionSwingTwist.Twist
    ) {template="DISPATCH_RigVMDispatch_SelectInt32(in Index,in Values,out Result)"}.Result
    RotationOrder=@n1.Rotation Order
   )
   > DISPATCH_RigVMDispatch_Print = RigVMDispatchNode(ScreenColor=(R=1,G=1,B=1,A=1), ScreenDuration=1.5, Enabled=@n1.debugRotation, Value=@RigVMFunction_MathQuaternionToEuler.Result) {template="DISPATCH_RigVMDispatch_Print(in Prefix,in Value,in Enabled,in ScreenDuration,in ScreenColor)"}
   Clamp = &RigVMFunction_MathIntClamp(Value=@n1.primAxis)
   DISPATCH_RigVMDispatch_SelectInt32 = RigVMDispatchNode(
    Values=(0,0,0)
    Index=@Clamp.Result
    Values.0=@RigVMFunction_MathQuaternionToEuler.Result.X
    Values.1=@RigVMFunction_MathQuaternionToEuler.Result.Y
    Values.2=@RigVMFunction_MathQuaternionToEuler.Result.Z
   ) {template="DISPATCH_RigVMDispatch_SelectInt32(in Index,in Values,out Result)"}
   Remap = &RigVMFunction_MathDoubleRemap(
    TargetMaximum=1
    TargetMinimum=0
    bClamp=False
    Value=@DISPATCH_RigVMDispatch_SelectInt32.Result
    SourceMaximum=@n1.max
    SourceMinimum=@n1.min
   )
   Multiply = &RigVMFunction_MathDoubleMul(A=&RigVMFunction_MathDoubleClamp(Value=@Remap.Result).Result, B=@n1.premult)
   > RigVMFunctionReturnNode(outputWeight=@Multiply.Result)
  ]}
 RigVMFunctionLibrary {links [], nodes []}
 RigVMFunction_Sequence_1_ContainedGraph {nodes [
   n1 = RigVMFunctionEntryNode()
   RigVMFunction_Sequence = &RigVMFunction_Sequence(ExecuteContext=@n1.ExecuteContext)
   n2 = &RigVMFunction_Sequence(ExecuteContext=@RigVMFunction_Sequence.B)
   RigVMFunctionReturnNode(A=@RigVMFunction_Sequence.A, B=@n2.A, C=@n2.B)
  ]}
 RigVMModel {nodes [
   RigUnit_BeginExecution = &RigUnit_BeginExecution()
   RigVMFunction_Sequence = RigVMAggregateNode(ExecuteContext=@RigUnit_BeginExecution.ExecutePin)
   &RigUnit_ParentConstraint(
    AdvancedSettings=(InterpolationType=Average,RotationOrderForFilter=XZY)
    Child=(Type=Bone,Name="base_scarf_c")
    Parents=((Item=(Type=Bone,Name="pack"),Weight=0.689),(Item=(Type=Bone,Name="dyn_head"),Weight=0.311))
    Weight=1
    ExecutePin=@RigVMFunction_Sequence.A
   )
   > &RigUnit_ParentConstraint(
    AdvancedSettings=(InterpolationType=Average,RotationOrderForFilter=XZY)
    Child=(Type=Bone,Name="base_scarf_l")
    Parents=((Item=(Type=Bone,Name="pack"),Weight=0.732),(Item=(Type=Bone,Name="dyn_head"),Weight=0.268))
    Weight=1
   )
   > &RigUnit_ParentConstraint(
    AdvancedSettings=(InterpolationType=Average,RotationOrderForFilter=XZY)
    Child=(Type=Bone,Name="base_scarf_r")
    Parents=((Item=(Type=Bone,Name="pack"),Weight=0.732),(Item=(Type=Bone,Name="dyn_head"),Weight=0.268))
    Weight=1
   )
   > &RigUnit_ParentConstraint(
    AdvancedSettings=(InterpolationType=Average,RotationOrderForFilter=XZY)
    Child=(Type=Bone,Name="base_knot")
    Parents=((Item=(Type=Bone,Name="pack"),Weight=0.617),(Item=(Type=Bone,Name="dyn_head"),Weight=0.383))
    Weight=1
   )
   DeltaRotation = RigVMCollapseNode(
    Rotation Order=ZYX
    child=dyn_head
    debugRotation=false
    max=-10
    parent=dyn_head_DeltaRef_null
    parentType=Null
    premult=1
    primAxis=2
    twistAxis=(X=1,Y=0,Z=0)
    ExecuteContext=@RigVMFunction_Sequence.B
   )
   > &RigUnit_ModifyTransforms(
    ItemToModify=((Item=(Type=Bone,Name="base_scarf_c"),Transform=(Rotation=(X=0,Y=-0.0872,Z=0,W=0.9962),Translation=(X=0,Y=0,Z=0),Scale3D=(X=1,Y=1,Z=1))))
    WeightMinimum=0
    Weight=@DeltaRotation.outputWeight
   )
   > &RigUnit_ModifyTransforms(
    ItemToModify=((Item=(Type=Bone,Name="base_scarf_l"),Transform=(Rotation=(X=0,Y=-0.0436,Z=0,W=0.999),Translation=(X=0,Y=0,Z=0),Scale3D=(X=1,Y=1,Z=1))))
    WeightMinimum=0
    Weight=@DeltaRotation.outputWeight
   )
   > &RigUnit_ModifyTransforms(
    ItemToModify=((Item=(Type=Bone,Name="base_scarf_r"),Transform=(Rotation=(X=0,Y=-0.0436,Z=0,W=0.999),Translation=(X=0,Y=0,Z=0),Scale3D=(X=1,Y=1,Z=1))))
    WeightMinimum=0
    Weight=@DeltaRotation.outputWeight
   )
   DeltaRotation_2 = RigVMCollapseNode(
    Rotation Order=ZYX
    child=dyn_leg_l
    debugRotation=false
    max=10
    min=0
    parent=dyn_leg_l_DeltaRef_null
    parentType=Null
    premult=1
    primAxis=2
    twistAxis=(X=1,Y=0,Z=0)
    ExecuteContext=@RigVMFunction_Sequence.C
   )
   > &RigUnit_ModifyTransforms(
    ItemToModify=((Item=(Type=Bone,Name="dyn_foot_l"),Transform=(Rotation=(X=0,Y=0.2588,Z=0,W=0.9659),Translation=(X=0,Y=0,Z=0),Scale3D=(X=1,Y=1,Z=1))))
    WeightMinimum=0
    Weight=@DeltaRotation_2.outputWeight
   )
  ]}
]
hierarchy [
 root {loc=<0 0 0>, rot=<0 0 0 1>}
 pack {parent=root, loc=<0 -0.7316 0>, rot=<0 0 0 1>}
 base_knot {parent=pack, loc=<0 -4.1937 6.1752>, rot=<0.3444 -0.3444 -0.6176 -0.6176>}
 dyn_knot {parent=base_knot, loc=<0 0 0>, rot=<0 0 0 1>}
 base_scarfKnot_l {parent=dyn_knot, loc=<0.2179 0.8446 0.6361>, rot=<-0.2435 -0.0723 -0.4823 -0.8384>}
 dyn_scarfKnot_l {parent=base_scarfKnot_l, loc=<0 0 0>, rot=<0 0 0 1>}
 base_scarfKnot_r {parent=dyn_knot, loc=<0.0915 -0.7743 0.5132>, rot=<0.1741 -0.1134 0.3815 -0.9007>}
 dyn_scarfKnot_r {parent=base_scarfKnot_r, loc=<0 0 0>, rot=<0 0 0 1>}
 base_head {parent=pack, loc=<-0.0015 -12.484 3.3076>, rot=<-0.1002 0.7 0.1002 -0.7>}
 dyn_head {parent=base_head, loc=<0 0 0>, rot=<0 0 0 1>}
 dyn_spike01 {parent=dyn_head, loc=<7.9392 2.7552 -0.0015>, rot=<0 0 -0.1865 -0.9824>}
 dyn_spike02 {parent=dyn_head, loc=<3.6521 7.3226 -0.0015>, rot=<0 0 -0.4947 -0.8691>}
 base_eye_r {parent=dyn_head, loc=<3.7051 -1.3191 -5.3524>, rot=<0.8458 -0.0561 -0.4862 0.2123>}
 dyn_eye_r {parent=base_eye_r, loc=<0 0 0>, rot=<0 0 0 1>}
 base_pupil_r {parent=dyn_eye_r, loc=<3.3847 -0.3028 -0.4389>, rot=<0 0 0 1>}
 dyn_pupil_r {parent=base_pupil_r, loc=<0 0 0>, rot=<0 0 0 1>}
 base_eye_l {parent=dyn_head, loc=<3.7051 -1.3191 5.3495>, rot=<0.8458 -0.0561 0.4862 -0.2123>}
 dyn_eye_l {parent=base_eye_l, loc=<0 0 0>, rot=<0 0 0 1>}
 base_pupil_l {parent=dyn_eye_l, loc=<3.3847 -0.3028 0.4389>, rot=<0 0 0 1>}
 dyn_pupil_l {parent=base_pupil_l, loc=<0 0 0>, rot=<0 0 0 1>}
 base_scarf_c {parent=pack, loc=<0.1493 -20.7785 1.3994>, rot=<-0.4729 -0.4814 0.5172 -0.5265>}
 dyn_scarf_c {parent=base_scarf_c, loc=<0 0 0>, rot=<0 0 0 1>}
 base_scarf_l {parent=pack, loc=<-6.4393 -18.0712 2.658>, rot=<-0.6384 -0.2468 0.662 -0.3054>}
 dyn_scarf_l {parent=base_scarf_l, loc=<0 0 0>, rot=<0 0 0 1>}
 base_scarf_r {parent=pack, loc=<6.5271 -17.8661 2.217>, rot=<-0.2354 -0.6427 0.2936 -0.6674>}
 dyn_scarf_r {parent=base_scarf_r, loc=<0 0 0>, rot=<0 0 0 1>}
 base_hip {parent=pack, loc=<-0.0015 -10.1599 -7.7246>, rot=<0 -0.7071 0 -0.7071>}
 dyn_hip {parent=base_hip, loc=<0 0 0>, rot=<0 0 0 1>}
 base_leg_l {parent=dyn_hip, loc=<4.7681 -6.6275 -5.6735>, rot=<0.6408 -0.3221 -0.5506 0.4272>}
 dyn_leg_l {parent=base_leg_l, loc=<0 0 0>, rot=<0 0 0 1>}
 dyn_foot_l {parent=dyn_leg_l, loc=<2.0083 0 0>, rot=<0 0 0 1>}
 base_leg_r {parent=dyn_hip, loc=<4.7681 -6.6275 5.6765>, rot=<-0.4082 0.5409 0.3381 -0.653>}
 dyn_leg_r {parent=base_leg_r, loc=<0 0 0>, rot=<0 0 0 1>}
 dyn_foot_r {parent=dyn_leg_r, loc=<2.0084 0 0>, rot=<0 0 0 -1>}
 base_tail_01 {parent=dyn_hip, loc=<4.8074 6.068 0.0015>, rot=<-0.705 0.0539 0.0539 -0.705>}
 dyn_tail_01 {parent=base_tail_01, loc=<0 0 0>, rot=<0 0 0 1>}
 dyn_tail_02 {parent=dyn_tail_01, loc=<3.9164 0 0>, rot=<0 -0.0047 0 -1>}
 dyn_leg_l_DeltaRef_null {parent=dyn_hip, type=Null, loc=<4.7681 -6.6275 -5.6735>, rot=<0.6408 -0.3221 -0.5506 0.4272>}
 dyn_leg_r_DeltaRef_null {parent=dyn_hip, type=Null, loc=<4.7681 -6.6275 5.6765>, rot=<-0.4082 0.5409 0.3381 -0.653>}
 dyn_head_DeltaRef_null {parent=pack, type=Null, loc=<-0.0015 -12.484 3.3076>, rot=<-0.1002 0.7 0.1002 -0.7>}
]
